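# Parent material for the DolphinGill gadget emote prop, adding colored lamp lights between the shared CharacterParent PreFX and PostFX functions.
# UV bands pick among four color parameters, while textures and channel masks supply lamp masks and IDs.
# Offset sine and cosine waves drive optional blinking, and Fresnel shapes lamp base color and emissive.
# Tiled sphere masks form page lamps, and a material-attribute blend layers the lamp look onto the base.
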Material /BRCosmetics/Animation/Props/Gadgets_DolphinGill/Materials/M_Gadget_DolphinGill_Parent
graph {
 expressions {
  MaterialExpressionAdd_0 {
   class=MaterialExpressionAdd
   inputs [
    A {from=MaterialExpressionMultiply_8}
    B {from=MaterialExpressionMultiply_7}
   ]
  }
  MaterialExpressionAdd_1 {
   class=MaterialExpressionAdd
   inputs [
    A {from=MaterialExpressionPower_2}
    B {from=MaterialExpressionMultiply_17}
   ]
  }
  MaterialExpressionAdd_2 {
   class=MaterialExpressionAdd
   inputs [
    A {from=MaterialExpressionSaturate_2}
    B {from=MaterialExpressionSaturate_3}
   ]
  }
  MaterialExpressionBlendMaterialAttributes_0 {
   class=MaterialExpressionBlendMaterialAttributes
   inputs [
    A {from=MaterialExpressionNamedRerouteUsage_4, output=""}
    B {from=MaterialExpressionSetMaterialAttributes_0, output=""}
    Alpha {from=MaterialExpressionMax_3, output=""}
   ]
   properties {}
  }
  MaterialExpressionCeil_0 {class=MaterialExpressionCeil, inputs [None {from=MaterialExpressionSaturate_6, output=""}], properties {}}
  MaterialExpressionChannelMaskParameter_0 {
   class=MaterialExpressionChannelMaskParameter
   inputs [None {from=MaterialExpressionNamedRerouteUsage_5, output=""}]
   properties {maskChannel=Blue, parameterName=LampFXMask, defaultValue=<rgba 0 0 1 0>}
  }
  MaterialExpressionChannelMaskParameter_1 {
   class=MaterialExpressionChannelMaskParameter
   inputs [None {from=MaterialExpressionNamedRerouteUsage_10, output=""}]
   properties {parameterName=PagesFXMask}
  }
  MaterialExpressionComponentMask_0 {class=MaterialExpressionComponentMask, inputs [None {from=MaterialExpressionTextureCoordinate_0}], properties {g=true}}
  MaterialExpressionComponentMask_1 {class=MaterialExpressionComponentMask, inputs [None {from=MaterialExpressionTextureCoordinate_1}], properties {g=true}}
  MaterialExpressionComponentMask_2 {class=MaterialExpressionComponentMask, inputs [None {from=MaterialExpressionTextureCoordinate_2}], properties {r=true}}
  MaterialExpressionComponentMask_3 {class=MaterialExpressionComponentMask, inputs [None {from=MaterialExpressionVectorParameter_4, output=RGB}], properties {g=true, r=true}}
  MaterialExpressionConstant3Vector_0 {class=MaterialExpressionConstant3Vector, properties {constant=<rgba 0.5 0.5 0.5 0>}}
  MaterialExpressionConstantBiasScale_2 {class=MaterialExpressionConstantBiasScale, inputs [None {from=MaterialExpressionLinearInterpolate_3}], properties {bias=-0.5, scale=2}}
  MaterialExpressionCosine_1 {class=MaterialExpressionCosine, inputs [None {from=MaterialExpressionAdd_0, output=""}], properties {period=1.3457}}
  MaterialExpressionDotProduct_0 {
   class=MaterialExpressionDotProduct
   inputs [
    A {from=MaterialExpressionTextureSample_0, output=RGB}
    B {from=MaterialExpressionConstant3Vector_0, output=""}
   ]
   properties {}
  }
  MaterialExpressionFloor_0 {class=MaterialExpressionFloor, inputs [None {from=MaterialExpressionMultiply_0, output=""}], properties {}}
  MaterialExpressionFrac_0 {class=MaterialExpressionFrac, inputs [None {from=MaterialExpressionMultiply_12, output=""}], properties {}}
  MaterialExpressionFrac_1 {class=MaterialExpressionFrac, inputs [None {from=MaterialExpressionMultiply_9, output=""}], properties {}}
  MaterialExpressionFrac_3 {class=MaterialExpressionFrac, inputs [None {from=MaterialExpressionMultiply_15, output=""}], properties {}}
  MaterialExpressionFresnel_0 {
   class=MaterialExpressionFresnel
   inputs [
    ExponentIn {from="", output=""}
    BaseReflectFractionIn {from="", output=""}
    Normal {from="", output=""}
   ]
   properties {baseReflectFraction=0, exponent=1}
  }
  MaterialExpressionFresnel_1 {
   class=MaterialExpressionFresnel
   inputs [
    ExponentIn {from="", output=""}
    BaseReflectFractionIn {from="", output=""}
    Normal {from="", output=""}
   ]
   properties {baseReflectFraction=0, exponent=0.2}
  }
  MaterialExpressionLinearInterpolate_0 {
   class=MaterialExpressionLinearInterpolate
   inputs [
    A {from=MaterialExpressionComponentMask_0}
    B {from=MaterialExpressionNamedRerouteUsage_11}
    Alpha {from=MaterialExpressionScalarParameter_0}
   ]
  }
  MaterialExpressionLinearInterpolate_1 {
   class=MaterialExpressionLinearInterpolate
   inputs [
    A {from=MaterialExpressionMultiply_2}
    B {from=MaterialExpressionMultiply_5}
    Alpha {from=MaterialExpressionNamedRerouteUsage_9}
   ]
  }
  MaterialExpressionLinearInterpolate_2 {
   class=MaterialExpressionLinearInterpolate
   inputs [
    A {from=MaterialExpressionAdd_2}
    B {from=MaterialExpressionSaturate_2}
    Alpha {from=MaterialExpressionScalarParameter_8}
   ]
  }
  MaterialExpressionLinearInterpolate_3 {
   class=MaterialExpressionLinearInterpolate
   inputs [
    A {from=MaterialExpressionNamedRerouteUsage_21}
    B {from=MaterialExpressionNamedRerouteUsage_0}
    Alpha {from=MaterialExpressionNamedRerouteUsage_16}
   ]
  }
  MaterialExpressionLinearInterpolate_4 {
   class=MaterialExpressionLinearInterpolate
   inputs [
    A {from=MaterialExpressionSwitch_0}
    B {from=MaterialExpressionNamedRerouteUsage_18}
    Alpha {from=MaterialExpressionNamedRerouteUsage_17}
   ]
  }
  MaterialExpressionLinearInterpolate_5 {
   class=MaterialExpressionLinearInterpolate
   inputs [
    A {from=""}
    B {from=MaterialExpressionSmoothStep_2}
    Alpha {from=MaterialExpressionScalarParameter_7}
   ]
   properties {constA=1}
  }
  MaterialExpressionLinearInterpolate_6 {
   class=MaterialExpressionLinearInterpolate
   inputs [
    A {from=MaterialExpressionStep_10}
    B {from=MaterialExpressionSine_1}
    Alpha {from=MaterialExpressionStep_11}
   ]
  }
  MaterialExpressionMaterialFunctionCall_0 {
   class=MaterialExpressionMaterialFunctionCall
   inputs [
    Overide Texture Set UVs {from=""}
    Override Tech Art Mask UVs {from=""}
    Override Basic Emissive UVs {from=""}
   ]
   properties {materialFunction=&MF_CharacterParent_2023_PreFX}
  }
  MaterialExpressionMaterialFunctionCall_1 {
   class=MaterialExpressionMaterialFunctionCall
   inputs [
    Material {from=MaterialExpressionNamedRerouteUsage_3}
    M {from=MaterialExpressionMaterialFunctionCall_0, output=M}
    Pre Skinned Local Position {from=MaterialExpressionMaterialFunctionCall_0, output=M}
    Distance Blend {from=MaterialExpressionMaterialFunctionCall_0, output=M}
    Fresnel {from=MaterialExpressionMaterialFunctionCall_0, output=M}
   ]
   properties {materialFunction=&MF_CharacterParent_2023_PostFX}
  }
  MaterialExpressionMax_0 {
   class=MaterialExpressionMax
   inputs [
    A {from=MaterialExpressionMultiply_6}
    B {from=""}
   ]
  }
  MaterialExpressionMax_1 {
   class=MaterialExpressionMax
   inputs [
    A {from=MaterialExpressionNamedRerouteUsage_23}
    B {from=MaterialExpressionNamedRerouteUsage_24}
   ]
  }
  MaterialExpressionMax_2 {
   class=MaterialExpressionMax
   inputs [
    A {from=MaterialExpressionMultiply_14}
    B {from=MaterialExpressionLinearInterpolate_4}
   ]
  }
  MaterialExpressionMax_3 {
   class=MaterialExpressionMax
   inputs [
    A {from=MaterialExpressionNamedRerouteUsage_6}
    B {from=MaterialExpressionNamedRerouteUsage_19}
   ]
  }
  MaterialExpressionMultiply_0 {
   class=MaterialExpressionMultiply
   inputs [
    A {from=MaterialExpressionComponentMask_0}
    B {from=MaterialExpressionScalarParameter_1}
   ]
   properties {constB=4}
  }
  MaterialExpressionMultiply_1 {
   class=MaterialExpressionMultiply
   inputs [
    A {from=MaterialExpressionPower_1}
    B {from=MaterialExpressionNamedRerouteUsage_7}
   ]
  }
  MaterialExpressionMultiply_10 {
   class=MaterialExpressionMultiply
   inputs [
    A {from=MaterialExpressionSaturate_5}
    B {from=MaterialExpressionStep_0}
   ]
  }
  MaterialExpressionMultiply_11 {
   class=MaterialExpressionMultiply
   inputs [
    A {from=MaterialExpressionTime_1}
    B {from=""}
   ]
   properties {constB=0.5}
  }
  MaterialExpressionMultiply_12 {
   class=MaterialExpressionMultiply
   inputs [
    A {from=MaterialExpressionTextureCoordinate_3}
    B {from=""}
   ]
   properties {constB=4}
  }
  MaterialExpressionMultiply_14 {
   class=MaterialExpressionMultiply
   inputs [
    A {from=MaterialExpressionNamedRerouteUsage_13}
    B {from=MaterialExpressionSwitch_0}
   ]
  }
  MaterialExpressionMultiply_15 {
   class=MaterialExpressionMultiply
   inputs [
    A {from=MaterialExpressionMultiply_11}
    B {from=""}
   ]
   properties {constB=0.3}
  }
  MaterialExpressionMultiply_16 {
   class=MaterialExpressionMultiply
   inputs [
    A {from=MaterialExpressionAdd_1}
    B {from=MaterialExpressionNamedRerouteUsage_15}
   ]
  }
  MaterialExpressionMultiply_17 {
   class=MaterialExpressionMultiply
   inputs [
    A {from=MaterialExpressionPower_2}
    B {from=MaterialExpressionScalarParameter_11}
   ]
  }
  MaterialExpressionMultiply_18 {
   class=MaterialExpressionMultiply
   inputs [
    A {from=MaterialExpressionMultiply_16}
    B {from=MaterialExpressionNamedRerouteUsage_20}
   ]
  }
  MaterialExpressionMultiply_2 {
   class=MaterialExpressionMultiply
   inputs [
    A {from=MaterialExpressionMultiply_1}
    B {from=MaterialExpressionScalarParameter_10}
   ]
  }
  MaterialExpressionMultiply_3 {
   class=MaterialExpressionMultiply
   inputs [
    A {from=MaterialExpressionLinearInterpolate_5}
    B {from=MaterialExpressionScalarParameter_6}
   ]
  }
  MaterialExpressionMultiply_4 {
   class=MaterialExpressionMultiply
   inputs [
    A {from=MaterialExpressionNamedRerouteDeclaration_7}
    B {from=MaterialExpressionOneMinus_2}
   ]
  }
  MaterialExpressionMultiply_5 {
   class=MaterialExpressionMultiply
   inputs [
    A {from=MaterialExpressionReroute_8}
    B {from=MaterialExpressionMax_0}
   ]
  }
  MaterialExpressionMultiply_6 {
   class=MaterialExpressionMultiply
   inputs [
    A {from=MaterialExpressionOneMinus_1}
    B {from=MaterialExpressionScalarParameter_4}
   ]
  }
  MaterialExpressionMultiply_7 {
   class=MaterialExpressionMultiply
   inputs [
    A {from=MaterialExpressionTime_0}
    B {from=MaterialExpressionScalarParameter_5}
   ]
  }
  MaterialExpressionMultiply_8 {
   class=MaterialExpressionMultiply
   inputs [
    A {from=MaterialExpressionConstantBiasScale_2}
    B {from=MaterialExpressionScalarParameter_21}
   ]
   properties {constB=0.25}
  }
  MaterialExpressionMultiply_9 {
   class=MaterialExpressionMultiply
   inputs [
    A {from=MaterialExpressionTime_1}
    B {from=""}
   ]
   properties {constB=0.3}
  }
  MaterialExpressionNamedRerouteDeclaration_0 {
   class=MaterialExpressionNamedRerouteDeclaration
   inputs [None {from=MaterialExpressionMaterialFunctionCall_0, output=Result}]
   properties {name=InMaterial, nodeColor=<rgba 0.4531 1 0 1>}
  }
  MaterialExpressionNamedRerouteDeclaration_1 {
   class=MaterialExpressionNamedRerouteDeclaration
   inputs [None {from=MaterialExpressionMultiply_10}]
   properties {name=MeshLampsID, nodeColor=<rgba 1 0 0.1875 1>}
  }
  MaterialExpressionNamedRerouteDeclaration_10 {
   class=MaterialExpressionNamedRerouteDeclaration
   inputs [None {from=MaterialExpressionCeil_0}]
   properties {name=MeshLampsMask, nodeColor=<rgba 0.9219 0 1 1>}
  }
  MaterialExpressionNamedRerouteDeclaration_11 {
   class=MaterialExpressionNamedRerouteDeclaration
   inputs [None {from=MaterialExpressionTextureSample_1, output=G}]
   properties {name=RandomID, nodeColor=<rgba 0 0.1797 1 1>}
  }
  MaterialExpressionNamedRerouteDeclaration_12 {
   class=MaterialExpressionNamedRerouteDeclaration
   inputs [None {from=MaterialExpressionTextureSample_1, output=R}]
   properties {name=CustomIDMap, nodeColor=<rgba 0.5 1 0 1>}
  }
  MaterialExpressionNamedRerouteDeclaration_13 {
   class=MaterialExpressionNamedRerouteDeclaration
   inputs [None {from=MaterialExpressionTextureSample_1, output=B}]
   properties {name=LampColorMask, nodeColor=<rgba 1 0 0.8203 1>}
  }
  MaterialExpressionNamedRerouteDeclaration_14 {
   class=MaterialExpressionNamedRerouteDeclaration
   inputs [None {from=MaterialExpressionSaturate_16}]
   properties {name=YellowLamps, nodeColor=<rgba 1 0.9259 0 1>}
  }
  MaterialExpressionNamedRerouteDeclaration_16 {
   class=MaterialExpressionNamedRerouteDeclaration
   inputs [None {from=MaterialExpressionChannelMaskParameter_1, output=R}]
   properties {name=PagesFXMask, nodeColor=<rgba 0 0.7187 1 1>}
  }
  MaterialExpressionNamedRerouteDeclaration_17 {
   class=MaterialExpressionNamedRerouteDeclaration
   inputs [None {from=MaterialExpressionMultiply_18}]
   properties {name=PageLamps, nodeColor=<rgba 1 0.9609 0 1>}
  }
  MaterialExpressionNamedRerouteDeclaration_2 {
   class=MaterialExpressionNamedRerouteDeclaration
   inputs [None {from=MaterialExpressionTextureSample_0, output=RGB}]
   properties {name=MeshLampColors, nodeColor=<rgba 0.7578 1 0 1>}
  }
  MaterialExpressionNamedRerouteDeclaration_20 {
   class=MaterialExpressionNamedRerouteDeclaration
   inputs [None {from=MaterialExpressionSaturate_0}]
   properties {name=GreenLamps, nodeColor=<rgba 0.074 1 0 1>}
  }
  MaterialExpressionNamedRerouteDeclaration_21 {
   class=MaterialExpressionNamedRerouteDeclaration
   inputs [None {from=MaterialExpressionSaturate_4}]
   properties {name=BlueLamps, nodeColor=<rgba 0 0.0338 1 1>}
  }
  MaterialExpressionNamedRerouteDeclaration_22 {
   class=MaterialExpressionNamedRerouteDeclaration
   inputs [None {from=MaterialExpressionSaturate_12}]
   properties {name=RedLamps, nodeColor=<rgba 1 0 0.0114 1>}
  }
  MaterialExpressionNamedRerouteDeclaration_3 {
   class=MaterialExpressionNamedRerouteDeclaration
   inputs [None {from=MaterialExpressionMultiply_4}]
   properties {name=OutBaseColor, nodeColor=<rgba 1 0.5859 0 1>}
  }
  MaterialExpressionNamedRerouteDeclaration_4 {
   class=MaterialExpressionNamedRerouteDeclaration
   inputs [None {from=MaterialExpressionBlendMaterialAttributes_0}]
   properties {name=OutMaterial, nodeColor=<rgba 0.2656 0 1 1>}
  }
  MaterialExpressionNamedRerouteDeclaration_5 {
   class=MaterialExpressionNamedRerouteDeclaration
   inputs [None {from=MaterialExpressionMaterialFunctionCall_0, output=Tech Art Mask}]
   properties {name=InFXMask, nodeColor=<rgba 0.875 0 1 1>}
  }
  MaterialExpressionNamedRerouteDeclaration_6 {
   class=MaterialExpressionNamedRerouteDeclaration
   inputs [None {from=MaterialExpressionChannelMaskParameter_0, output=R}]
   properties {name=LampMask, nodeColor=<rgba 0 1 0.5547 1>}
  }
  MaterialExpressionNamedRerouteDeclaration_7 {
   class=MaterialExpressionNamedRerouteDeclaration
   inputs [None {from=MaterialExpressionLinearInterpolate_4}]
   properties {name=LampColors, nodeColor=<rgba 0.8516 0 1 1>}
  }
  MaterialExpressionNamedRerouteDeclaration_8 {
   class=MaterialExpressionNamedRerouteDeclaration
   inputs [None {from=MaterialExpressionLinearInterpolate_1}]
   properties {name=OutEmissive, nodeColor=<rgba 0 1 0.5313 1>}
  }
  MaterialExpressionNamedRerouteDeclaration_9 {
   class=MaterialExpressionNamedRerouteDeclaration
   inputs [None {from=MaterialExpressionMultiply_3}]
   properties {name=LightsOn, nodeColor=<rgba 0.0312 1 0 1>}
  }
  MaterialExpressionNamedRerouteUsage_0 {class=MaterialExpressionNamedRerouteUsage}
  MaterialExpressionNamedRerouteUsage_1 {class=MaterialExpressionNamedRerouteUsage}
  MaterialExpressionNamedRerouteUsage_10 {class=MaterialExpressionNamedRerouteUsage}
  MaterialExpressionNamedRerouteUsage_11 {class=MaterialExpressionNamedRerouteUsage}
  MaterialExpressionNamedRerouteUsage_12 {class=MaterialExpressionNamedRerouteUsage}
  MaterialExpressionNamedRerouteUsage_13 {class=MaterialExpressionNamedRerouteUsage}
  MaterialExpressionNamedRerouteUsage_14 {class=MaterialExpressionNamedRerouteUsage}
  MaterialExpressionNamedRerouteUsage_15 {class=MaterialExpressionNamedRerouteUsage}
  MaterialExpressionNamedRerouteUsage_16 {class=MaterialExpressionNamedRerouteUsage}
  MaterialExpressionNamedRerouteUsage_17 {class=MaterialExpressionNamedRerouteUsage}
  MaterialExpressionNamedRerouteUsage_18 {class=MaterialExpressionNamedRerouteUsage}
  MaterialExpressionNamedRerouteUsage_19 {class=MaterialExpressionNamedRerouteUsage}
  MaterialExpressionNamedRerouteUsage_2 {class=MaterialExpressionNamedRerouteUsage}
  MaterialExpressionNamedRerouteUsage_20 {class=MaterialExpressionNamedRerouteUsage}
  MaterialExpressionNamedRerouteUsage_21 {class=MaterialExpressionNamedRerouteUsage}
  MaterialExpressionNamedRerouteUsage_23 {class=MaterialExpressionNamedRerouteUsage}
  MaterialExpressionNamedRerouteUsage_24 {class=MaterialExpressionNamedRerouteUsage}
  MaterialExpressionNamedRerouteUsage_3 {class=MaterialExpressionNamedRerouteUsage}
  MaterialExpressionNamedRerouteUsage_4 {class=MaterialExpressionNamedRerouteUsage}
  MaterialExpressionNamedRerouteUsage_5 {class=MaterialExpressionNamedRerouteUsage}
  MaterialExpressionNamedRerouteUsage_6 {class=MaterialExpressionNamedRerouteUsage}
  MaterialExpressionNamedRerouteUsage_7 {class=MaterialExpressionNamedRerouteUsage}
  MaterialExpressionNamedRerouteUsage_8 {class=MaterialExpressionNamedRerouteUsage}
  MaterialExpressionNamedRerouteUsage_9 {class=MaterialExpressionNamedRerouteUsage}
  MaterialExpressionOneMinus_1 {class=MaterialExpressionOneMinus, inputs [None {from=MaterialExpressionFresnel_0}]}
  MaterialExpressionOneMinus_2 {class=MaterialExpressionOneMinus, inputs [None {from=MaterialExpressionFresnel_1}]}
  MaterialExpressionPower_0 {
   class=MaterialExpressionPower
   inputs [
    Base {from=MaterialExpressionDotProduct_0}
    Exp {from=""}
   ]
   properties {constExponent=4}
  }
  MaterialExpressionPower_1 {
   class=MaterialExpressionPower
   inputs [
    Base {from=MaterialExpressionFresnel_0}
    Exp {from=MaterialExpressionScalarParameter_12}
   ]
  }
  MaterialExpressionPower_2 {
   class=MaterialExpressionPower
   inputs [
    Base {from=MaterialExpressionSphereMask_0}
    Exp {from=MaterialExpressionScalarParameter_9}
   ]
   properties {constExponent=4}
  }
  MaterialExpressionReroute_8 {class=MaterialExpressionReroute, inputs [None {from=MaterialExpressionNamedRerouteUsage_7}]}
  MaterialExpressionSaturate_0 {class=MaterialExpressionSaturate, inputs [None {from=MaterialExpressionSubtract_0}]}
  MaterialExpressionSaturate_1 {class=MaterialExpressionSaturate, inputs [None {from=MaterialExpressionLinearInterpolate_2}]}
  MaterialExpressionSaturate_12 {class=MaterialExpressionSaturate, inputs [None {from=MaterialExpressionStep_7}]}
  MaterialExpressionSaturate_16 {class=MaterialExpressionSaturate, inputs [None {from=MaterialExpressionStep_4}]}
  MaterialExpressionSaturate_2 {class=MaterialExpressionSaturate, inputs [None {from=MaterialExpressionSine_0}]}
  MaterialExpressionSaturate_3 {class=MaterialExpressionSaturate, inputs [None {from=MaterialExpressionCosine_1}]}
  MaterialExpressionSaturate_4 {class=MaterialExpressionSaturate, inputs [None {from=MaterialExpressionSubtract_1}]}
  MaterialExpressionSaturate_5 {class=MaterialExpressionSaturate, inputs [None {from=MaterialExpressionPower_0}]}
  MaterialExpressionSaturate_6 {class=MaterialExpressionSaturate, inputs [None {from=MaterialExpressionMultiply_10}]}
  MaterialExpressionScalarParameter_0 {class=MaterialExpressionScalarParameter, properties {parameterName=UseUV2ForColorIDs}}
  MaterialExpressionScalarParameter_1 {class=MaterialExpressionScalarParameter, properties {defaultValue=4, parameterName=LampsAmount}}
  MaterialExpressionScalarParameter_10 {class=MaterialExpressionScalarParameter, properties {defaultValue=0.25, parameterName=LightsOffEmissiveIntensity}}
  MaterialExpressionScalarParameter_11 {class=MaterialExpressionScalarParameter, properties {defaultValue=1.6774, parameterName=PageLampsIntensity}}
  MaterialExpressionScalarParameter_12 {class=MaterialExpressionScalarParameter, properties {defaultValue=1, parameterName=LightsOffEmissivePower}}
  MaterialExpressionScalarParameter_13 {class=MaterialExpressionScalarParameter, properties {defaultValue=1, parameterName=PageLightRadius}}
  MaterialExpressionScalarParameter_15 {class=MaterialExpressionScalarParameter, properties {defaultValue=0.91, parameterName=LightsBlinkThreshold}}
  MaterialExpressionScalarParameter_2 {class=MaterialExpressionScalarParameter, properties {defaultValue=0.3, parameterName=Metalness}}
  MaterialExpressionScalarParameter_21 {class=MaterialExpressionScalarParameter, properties {defaultValue=1, parameterName=LightsRandomAmount}}
  MaterialExpressionScalarParameter_3 {class=MaterialExpressionScalarParameter, properties {defaultValue=0.2, parameterName=Roughness}}
  MaterialExpressionScalarParameter_4 {class=MaterialExpressionScalarParameter, properties {defaultValue=4, parameterName=LampIntensity}}
  MaterialExpressionScalarParameter_5 {class=MaterialExpressionScalarParameter, properties {defaultValue=0.4, parameterName=LightsSpeed}}
  MaterialExpressionScalarParameter_6 {class=MaterialExpressionScalarParameter, properties {defaultValue=1, parameterName=LightsOn}}
  MaterialExpressionScalarParameter_7 {class=MaterialExpressionScalarParameter, properties {defaultValue=1, parameterName=LightsEnableBlink}}
  MaterialExpressionScalarParameter_8 {class=MaterialExpressionScalarParameter, properties {parameterName=LightsLinearFlicker}}
  MaterialExpressionScalarParameter_9 {class=MaterialExpressionScalarParameter, properties {defaultValue=5.8508, parameterName=PageLampSmoothness}}
  MaterialExpressionSetMaterialAttributes_0 {
   class=MaterialExpressionSetMaterialAttributes
   inputs [
    MaterialAttributes {from=MaterialExpressionNamedRerouteUsage_1, output=""}
    Base Color {from=MaterialExpressionNamedRerouteUsage_2, output=""}
    Emissive Color {from=MaterialExpressionNamedRerouteUsage_8, output=""}
    Metallic {from=MaterialExpressionScalarParameter_2, output=""}
    Roughness {from=MaterialExpressionScalarParameter_3, output=""}
   ]
   properties {attributeSetTypes [
     #1
     #2
     #3
     #4
    ]}
  }
  MaterialExpressionSine_0 {class=MaterialExpressionSine, inputs [None {from=MaterialExpressionAdd_0}], properties {period=1.7435}}
  MaterialExpressionSine_1 {class=MaterialExpressionSine, inputs [None {from=MaterialExpressionMultiply_11}]}
  MaterialExpressionSmoothStep_2 {
   class=MaterialExpressionSmoothStep
   inputs [
    Min {from=MaterialExpressionScalarParameter_15, output=""}
    Max {from="", output=""}
    Value {from=MaterialExpressionSaturate_1, output=""}
   ]
   properties {constMin=0.8}
  }
  MaterialExpressionSphereMask_0 {
   class=MaterialExpressionSphereMask
   inputs [
    A {from=MaterialExpressionFrac_0, output=""}
    B {from=MaterialExpressionComponentMask_3, output=""}
    Radius {from=MaterialExpressionScalarParameter_13, output=""}
    Hardness {from="", output=""}
   ]
   properties {attenuationRadius=0.5, hardnessPercent=0}
  }
  MaterialExpressionStep_0 {
   class=MaterialExpressionStep
   inputs [
    Y {from="", output=""}
    X {from=MaterialExpressionComponentMask_1, output=""}
   ]
   properties {constX=0.5, constY=0.5}
  }
  MaterialExpressionStep_10 {
   class=MaterialExpressionStep
   inputs [
    Y {from="", output=""}
    X {from=MaterialExpressionSine_1, output=""}
   ]
   properties {constX=0.01, constY=0.01}
  }
  MaterialExpressionStep_11 {
   class=MaterialExpressionStep
   inputs [
    Y {from="", output=""}
    X {from=MaterialExpressionComponentMask_2, output=""}
   ]
   properties {constX=0.5, constY=0.5}
  }
  MaterialExpressionStep_4 {
   class=MaterialExpressionStep
   inputs [
    Y {from=MaterialExpressionNamedRerouteUsage_12, output=""}
    X {from="", output=""}
   ]
   properties {constX=0.25, constY=0.3}
  }
  MaterialExpressionStep_5 {
   class=MaterialExpressionStep
   inputs [
    Y {from=MaterialExpressionNamedRerouteUsage_12, output=""}
    X {from="", output=""}
   ]
   properties {constX=0.6, constY=0.3}
  }
  MaterialExpressionStep_6 {
   class=MaterialExpressionStep
   inputs [
    Y {from=MaterialExpressionNamedRerouteUsage_12, output=""}
    X {from="", output=""}
   ]
   properties {constX=0.8, constY=0.3}
  }
  MaterialExpressionStep_7 {
   class=MaterialExpressionStep
   inputs [
    Y {from="", output=""}
    X {from=MaterialExpressionNamedRerouteUsage_12, output=""}
   ]
   properties {constY=1}
  }
  MaterialExpressionSubtract_0 {
   class=MaterialExpressionSubtract
   inputs [
    A {from=MaterialExpressionStep_5}
    B {from=MaterialExpressionNamedRerouteUsage_14}
   ]
  }
  MaterialExpressionSubtract_1 {
   class=MaterialExpressionSubtract
   inputs [
    A {from=MaterialExpressionStep_6}
    B {from=MaterialExpressionMax_1}
   ]
  }
  MaterialExpressionSwitch_0 {
   class=MaterialExpressionSwitch
   inputs [
    SwitchValue {from=MaterialExpressionFloor_0, output=""}
    Default {from="", output=""}
    None {from=MaterialExpressionVectorParameter_0, output=RGB}
    None {from=MaterialExpressionVectorParameter_1, output=RGB}
    None {from=MaterialExpressionVectorParameter_2, output=RGB}
    None {from=MaterialExpressionVectorParameter_3, output=RGB}
   ]
   properties {inputs [
     {
      inputName=None
      input {
       expression=&self:MaterialExpressionVectorParameter_0
       inputName=None
       mask=1
       maskA=0
       maskB=1
       maskG=1
       maskR=1
       outputIndex=0
      }
     }
     {
      inputName=None
      input {
       expression=&self:MaterialExpressionVectorParameter_1
       inputName=None
       mask=1
       maskA=0
       maskB=1
       maskG=1
       maskR=1
       outputIndex=0
      }
     }
     {
      inputName=None
      input {
       expression=&self:MaterialExpressionVectorParameter_2
       inputName=None
       mask=1
       maskA=0
       maskB=1
       maskG=1
       maskR=1
       outputIndex=0
      }
     }
     {
      inputName=None
      input {
       expression=&self:MaterialExpressionVectorParameter_3
       inputName=None
       mask=1
       maskA=0
       maskB=1
       maskG=1
       maskR=1
       outputIndex=0
      }
     }
    ]}
  }
  MaterialExpressionTextureCoordinate_0 {class=MaterialExpressionTextureCoordinate, properties {coordinateIndex=1}}
  MaterialExpressionTextureCoordinate_1 {class=MaterialExpressionTextureCoordinate}
  MaterialExpressionTextureCoordinate_2 {class=MaterialExpressionTextureCoordinate}
  MaterialExpressionTextureCoordinate_3 {class=MaterialExpressionTextureCoordinate, properties {coordinateIndex=1}}
  MaterialExpressionTextureSample_0 {
   class=MaterialExpressionTextureSample
   inputs [
    UVs {from=""}
    Tex {from=""}
    Apply View MipBias {from=""}
   ]
   properties {texture=/BRCosmetics/Animation/Props/Gadgets_DolphinGill/Textures/T_DolphinGill_Emote_Prop_E}
  }
  MaterialExpressionTextureSample_1 {
   class=MaterialExpressionTextureSample
   inputs [
    UVs {from=""}
    Tex {from=""}
    Apply View MipBias {from=""}
   ]
   properties {
    constCoordinate=1
    samplerType=SAMPLERTYPE_LinearColor
    texture=/BRCosmetics/Accessories/FORT_Backpacks/Backpack_M_MED_LanternFit/Textures/T_LanternFit_IDs
   }
  }
  MaterialExpressionTime_0 {class=MaterialExpressionTime}
  MaterialExpressionTime_1 {class=MaterialExpressionTime}
  MaterialExpressionVectorParameter_0 {class=MaterialExpressionVectorParameter, properties {parameterName=Lamp1Color, defaultValue=<rgba 0.8823 1 0.0077 1>}}
  MaterialExpressionVectorParameter_1 {class=MaterialExpressionVectorParameter, properties {parameterName=Lamp2Color, defaultValue=<rgba 0 1 0.1376 1>}}
  MaterialExpressionVectorParameter_2 {class=MaterialExpressionVectorParameter, properties {parameterName=Lamp3Color, defaultValue=<rgba 0.0005 0.1565 1 1>}}
  MaterialExpressionVectorParameter_3 {class=MaterialExpressionVectorParameter, properties {parameterName=Lamp4Color, defaultValue=<rgba 1 0.01 0 1>}}
  MaterialExpressionVectorParameter_4 {class=MaterialExpressionVectorParameter, properties {defaultValue=<rgba 0.532 0.2139 0 1>}}
 }
 properties {
  bUseMaterialAttributes=true
  bUsedWithMeshDeformer=true
  bUsedWithMorphTargets=true
  bUsedWithSkeletalMesh=true
  pixelDepthOffsetMode=PDOM_Legacy
  previewMesh=/BRCosmetics/Animation/Props/Gadgets_DolphinGill/Meshes/DolphinGill_Gadget
  shadingModel=MSM_FromMaterialExpression
  usedShadingModels=Default Lit | Subsurface | Clear Coat | Subsurface Profile | Cloth
 }
}
